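; Map marker actor for the pirate booty chest, revealing a sky beacon and X decal.
; UnhideSkyMarker, noted as called by the chest's passive ability, runs only while ShouldHide is true, then clears it, unhides the decal, plays the activation sound and shows the Niagara indicator.
; A timeline animates dynamic materials for the glowing X, pillar mesh and decal.
; On finish, it re-hides the decal only if ShouldHide is true again.
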(event EventBeginPlay
  (|CallParentFunction) ; This node is disabled and will not be called.\nDrag off pins to build functionality.
)

(event Collision|EventActorBeginOverlap
  (params (OtherActor :type Actor%20Object%20Reference)))

(event EventTick
  (params (DeltaSeconds :type Float%20%28single-precision%29)))

(event EventUnhideSkyMarker
  (bind _ns_piratechest_indicator (Variables|Default|GetNS_PirateChest_Indicator))
  (bind _mid_glowy_x (Variables|Default|GetMID_Glowy_X))
  ;[1 1 1] Hide and Unhide then sky Marker. This is called from the passive ability that opperates the chest.
  (if (Variables|Default|GetShouldHide)
    (Development|SetHiddeninGame (Variables|Default|GetDecal) false false)
    (Audio|PlaySoundatLocation "/Game/Sounds/Foley_Loot/Containers/Pirate_Booty/Pirate_Booty_Beacon_Activate_Cue.Pirate_Booty_Beacon_Activate_Cue" (Transformation|GetActorLocation))
    (Variables|Default|SetShouldHide false)
    (Niagara|SetNiagaraVariable(Bool) _ns_piratechest_indicator "IsVisible" true)
    (Rendering|SetVisibility _ns_piratechest_indicator true true)
    (bind (_direction _opacity _spawnscale _spawnverticalscale _scale_x_mark_on_z _scale_x_mark_on_xy _scalecorona _pillarscale _secondaryopacity _opacity_x) (Timeline|Timeline PlayFromStart)
      (:Update
        (Rendering|Material|SetScalarParameterValue _mid_glowy_x "Scale_Z" _scale_x_mark_on_z :!class /Script/Engine.MaterialInstanceDynamic)
        (Rendering|Material|SetScalarParameterValue _mid_glowy_x "Scale_XY" _scale_x_mark_on_xy :!class /Script/Engine.MaterialInstanceDynamic)
        (Rendering|Material|SetScalarParameterValue (Variables|Default|GetMID_Glow_Pillar_Mesh) "MaxScaleIncrease" _pillarscale :!class /Script/Engine.MaterialInstanceDynamic)
        (Rendering|Material|SetScalarParameterValue (Variables|Default|GetMI_Decal) "X Decal Fire Lerp" _opacity_x :!class /Script/Engine.MaterialInstanceDynamic))
      (:Finished
        (if (Variables|Default|GetShouldHide)
          (Development|SetHiddeninGame (Variables|Default|GetDecal) true false))))))

(fn ConstructionScript
  (|CallParentFunction))
